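; Box Factory variant of the keycard lock that removes tagged Kill Z volumes when its door unlock activity starts.
; On BeginPlay it calls the parent, then binds DestroyKillZVolume to the door-unlocked activity event.
; That handler runs only where cosmetic events cannot execute, such as a dedicated server.
; It destroys every KillZVolume matching the configured tag, logging each one or reporting when none is found.
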(event EventBeginPlay
  (|CallParentFunction)
  (Default|BindEventtoStartDoorUnlockedActivity (delegate Custom|DestroyKillZVolume)))

(event Custom|DestroyKillZVolume
  (params (LastInteractingPawn :type Fort%20Pawn%20Object%20Reference))
  (Networking|CanExecuteCosmeticEvents
    (:False
      (bind _outactors (Actor|GetAllActorsOfClasswithTag "/Script/Engine.KillZVolume" (Variables|Default|GetKillZVolumeTag)))
      (if (Utilities|Array|IsEmpty _outactors)
        (Fortnite|FortShippingLog self (Utilities|String|Append "Couldn't find Kill Z volume with tag " (Utilities|String|ToString(Name) (Variables|Default|GetKillZVolumeTag))) true)
        (else
          (for _array_element _outactors
            (Utilities|IsValid _array_element
              (:Is%20Valid
                (Fortnite|FortShippingLog self (Utilities|String|Append "Destroying Kill Z Volume: " (Utilities|GetDisplayName _array_element)))
                (Actor|DestroyActor _array_element)))))))))

(fn ConstructionScript
  (|CallParentFunction))
