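; Parent keycard lock actor that lets a player holding a point-of-interest keycard unlock an attached vault door.
; A gameplay tag check drives the prompt, and interacting applies a swipe-montage effect to the pawn.
; Unlocking consumes the keycard, fires a success cue, then opens the vault on a timer and reports quest progress.
; Success and failure drive a child-actor scanner screen and console emissive; hide-and-kill is overridden so the lock persists.
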(event EventHideAndKill
  ;[1 1 0] Override so the lock parent can never tell this to die
  (Development|PrintString "Somoene told me to KMS but I said nooo" false false))

(event EventBeginPlay
  ;[0 1 0] BeginPlay
  (|CallParentFunction)
  (bind _asbp_ui_scanner_screen (Utilities|Casting|CastToBP_UI_ScannerScreen (Variables|/Script/Engine.ChildActorComponent|GetChildActor (Variables|Default|GetScannerScreenActor))))
  (Variables|Screen|SetScannerScreen _asbp_ui_scanner_screen)
  (bind _returnvalue (Rendering|Material|CreateDynamicMaterialInstance (Variables|Default|GetLock) 0 0 "None" :!class /Script/Engine.PrimitiveComponent))
  (Utilities|IsValid _returnvalue
    (:Is%20Valid
      (bind _output_get (Variables|Screen|SetMIDI_ConsoleScreen _returnvalue))
      (bind _returnvalue_2 (Rendering|Material|GetVectorParameterValue _output_get (Variables|Screen|GetScreenMaterialParameter) :!class /Script/Engine.MaterialInstanceDynamic))
      (Variables|Screen|SetOriginalEmissive _returnvalue_2))))

(event Interaction|EventBlueprintOnInteract
  (params (InteractingPawn :type Fort%20Pawn%20Object%20Reference) (InteractionBeingAttempted :type EInteractionBeingAttempted%20Enum :default "FirstInteraction"))
  ;[1 1 1] Blueprint on Interact
  (|CallParentFunction InteractingPawn InteractionBeingAttempted)
  (GameplayEffects|ApplyGameplayEffectToTarget :self (Ability|GetFortAbilitySystemComponent InteractingPawn) :GameplayEffectClass "/Game/Athena/Items/EnvironmentalItems/Locks/Keycard/Abilities/GE_Grant_Swipe_Montage_Ability.GE_Grant_Swipe_Montage_Ability_C" :Target (Ability|GetFortAbilitySystemComponent InteractingPawn) :Level 0.0))

(event EventSetLock
  (params (LockState :type Enum_Athena_Lock%20Enum :default "NewEnumerator0"))
  ;[1 1 1] Play success stinger and then open door
  (Variables|LockParent|SetLockState "NewEnumerator2")
  (Ability|ExecuteGameplayCueReplicated (Variables|Default|GetGC_Success))
  (Inventory|RemoveItemFromPlayer :PlayerController (Fortnite|GetFortPlayerControllerfromActor (Variables|LockParentInteract|GetLastInteractingPawn)) :ItemDefinition (Variables|Default|GetKeyID) :AmountToRemove -1 :bForceRemoval true :bSkipVerbRemoval false)
  (Variables|LockParentInteract|SetMaxInteractAngle 0.0)
  (Utilities|Time|SetTimerbyEvent (delegate Custom|CallOpenVault) 1.0)
  (CallFunction|ShowSuccessScreen)
  (Default|CallStartDoorUnlockedActivity (Variables|LockParentInteract|GetLastInteractingPawn)))

(event Custom|CallOpenVault
  (bind _lastinteractingpawn (Variables|LockParentInteract|GetLastInteractingPawn))
  (bind _returnvalue (Fortnite|GetFortPlayerControllerfromActor _lastinteractingpawn))
  ;[1 1 1] Play success stinger and then open door
  (bind _asvault_door (Utilities|Casting|CastToVault_Door (Variables|LockParent|GetAttachedLockActor)))
  (Class|VaultDoor|VaultOpen _asvault_door)
  ;[1 1 1] Send quest progress
  (Quest|SendComplexCustomStatEvent (Quests|GetQuestManager _returnvalue "Invalid") _asvault_door "(GameplayTags=((TagName=\"Athena.Quests.UnlockFaction.Door\")))" (GameplayTags|GetOwnedGameplayTags _asvault_door :!class /Script/GameplayTags.GameplayTagAssetInterface) 1)
  ;[1 1 1] Send Custom QS2 Progress
  (CustomVerb|BroadcastCustomVerbMessage (CustomVerb|MakeCustomVerbMessage _returnvalue "(TagName=\"CustomVerb.OpenVault\")")))

(event EventApplyFailCheckEffect
  ;[1 1 1] Play fail stinger - logic for sounding alarms and doing damage, etc would go here
  (Ability|ExecuteGameplayCueReplicated (Variables|Default|GetGC_Fail))
  (CallFunction|ShowFailScreen))

(event Custom|ShowSuccessScreen
  ;[1 0 1] Screen success and failure
  (Variables|Screen|SetScreenWidgetActive true)
  (Rendering|Material|SetVectorParameterValue (Variables|Screen|GetMIDI_ConsoleScreen) (Variables|Screen|GetScreenMaterialParameter) (Variables|Screen|GetScreenSuccessEmissive) :!class /Script/Engine.MaterialInstanceDynamic)
  (Analytics|FireAnalyticEventEnvironmentalItemUsed (Fortnite|GetFortPlayerControllerfromActor (Variables|LockParentInteract|GetLastInteractingPawn)) self)
  (Class|BPUIScannerScreen|ScreenEventSuccess (Variables|Screen|GetScannerScreen) "")
  (Utilities|FlowControl|Delay (Variables|Screen|GetScreenDisplayDuration))
  (Rendering|Material|SetVectorParameterValue (Variables|Screen|GetMIDI_ConsoleScreen) (Variables|Screen|GetScreenMaterialParameter) (Variables|Screen|GetScreenDisabledEmissive) :!class /Script/Engine.MaterialInstanceDynamic))

(event Custom|ShowFailScreen
  ;[1 0 1] Screen success and failure
  (if (not (Variables|Screen|GetScreenWidgetActive))
    (Variables|Screen|SetScreenWidgetActive true)
    (Class|BPUIScannerScreen|ScreenEventFailure (Variables|Screen|GetScannerScreen) "")
    (Rendering|Material|SetVectorParameterValue (Variables|Screen|GetMIDI_ConsoleScreen) (Variables|Screen|GetScreenMaterialParameter) (Variables|Screen|GetScreenFailureEmissive) :!class /Script/Engine.MaterialInstanceDynamic)
    (Utilities|FlowControl|Delay (Variables|Screen|GetScreenDisplayDuration))
    (Rendering|Material|SetVectorParameterValue (Variables|Screen|GetMIDI_ConsoleScreen) (Variables|Screen|GetScreenMaterialParameter) (Variables|Screen|GetOriginalEmissive) :!class /Script/Engine.MaterialInstanceDynamic)
    (Variables|Screen|SetScreenWidgetActive false)))

(fn BlueprintCanInteract
  (params (InteractingPawn :type Fort%20Pawn%20Object%20Reference) (InteractionBeingAttempted :type EInteractionBeingAttempted%20Enum :default "FirstInteraction") (InteractionType :type TInteractionType%20Enum :default "IT_NoInteraction"))
  (return (and (|CallParentFunction InteractingPawn InteractionBeingAttempted) (Utilities|Enum|Equal(Enum) (Variables|LockParent|GetLockState) "NewEnumerator1"))))

(fn BlueprintGetInteractionString
  (params (InteractingPawn :type Fort%20Pawn%20Object%20Reference) (InteractionBeingAttempted :type EInteractionBeingAttempted%20Enum :default "FirstInteraction"))
  (|CallParentFunction InteractingPawn InteractionBeingAttempted)
  (bind _asbga_athena_lock_parent (Utilities|Casting|CastToBGA_Athena_Lock_Parent self)) ; This guy is a liar, we do need to cast to the parent
  (bind _checkpassed (CallFunction|LockCheck _asbga_athena_lock_parent InteractingPawn InteractionBeingAttempted))
  (if _checkpassed
    (return (Utilities|Text|FormatText "NSLOCTEXT(\"[C6D288A09386F0153856F6FF0088AD52]\", \"230E76264F073E3800989094E3FC2B99\", \"Swipe {PoI Name} Keycard\")" (Variables|Default|GetPoINameText)))
    (else
      (return (Utilities|Text|FormatText "NSLOCTEXT(\"[C6D288A09386F0153856F6FF0088AD52]\", \"484E179640B9036F9035859F115562D5\", \"Requires {PoI Name} Keycard\")" (Variables|Default|GetPoINameText))))))

(fn LockCheck
  (params (InteractingPawn :type Fort%20Pawn%20Object%20Reference) (InteractionBeingAttempted :type EInteractionBeingAttempted%20Enum :default "FirstInteraction"))
  (return (GameplayTags|HasMatchingGameplayTag InteractingPawn (Variables|Default|GetTagToCheck))))

(fn ConstructionScript
  (|CallParentFunction))
